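; Drivable drift variant of the Fortnite Valet basic car, adding drift, boost, gear-shift and light effects.
; A grounded handbrake applies a drift GameplayEffect to occupants and binds the catch-up delegate, which fires a timed boost with Niagara, cue, rumble and FOV kick.
; Gear changes play force feedback and camera shake, with an FOV dip on upshifts.
; A replicated counter updates client headlight color, while Niagara parameters color lens flare and taillights.
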(event EventBeginPlay
  ;[1 1 1] Begin play events
  (|CallParentFunction)
  (CallFunction|OnBeginPlayFX)
  (Class|FortAsyncActionGameStateReady|GameStateReadyAsync
    (:OnReady
      (Vehicle|AssignOnCurrentGearChangedDelegate (EventDispatchers|CreateEvent :!event Custom|HandleGearChange)))))

(event Custom|HandleGearChange
  (params (NewGear :type Integer))
  ;[1 1 1] Event ----> Handle any gear change fx/force feedback
  (bind _upshift (CallFunction|HandleShifting NewGear))
  ;; HandleFOVDuringShift
  (if _upshift
    (Variables|Camera|SetCameraFOVOffset -5.0)
    (Utilities|FlowControl|Delay 0.2)
    (Variables|Camera|SetCameraFOVOffset 0.0)))

(event Custom|OnStartHandBrakeCatchupForceDelegate_Event
  (bind _driverpawn (Variables|ValetDefaults|GetDriverPawn))
  (bind _returnvalue (Fortnite|GetFortPlayerControllerfromActor _driverpawn))
  ;[1 1 1] Event ---> Out of drift
  (Utilities|FlowControl|Sequence
    (:then_0
      (bind _returnvalue_2 (Niagara|SpawnSystemAttached (Variables|Drift|GetFXDriftingBoost) (Variables|Vehicle|GetSkeletalMesh) "root" "0, 0, 0" "0.000000,180.000000,0.000000" "SnapToTarget" true))
      (Niagara|SetNiagaraVariable(Bool) _returnvalue_2 "Mod_OffroadTires" (Vehicle|Mod|Tire|HasTireMod))
      (Niagara|SetNiagaraVariable(Bool) _returnvalue_2 "TireIntactRL" (Variables|FX|NiagaraSettings|GetTire_Popped_RL))
      (Niagara|SetNiagaraVariable(Bool) _returnvalue_2 "TireIntactRR" (Variables|FX|NiagaraSettings|GetTire_Popped_RR))
      (Niagara|SetNiagaraVariable(Int32) _returnvalue_2 "SedanColor" (Variables|Default|GetDriftingBoost_SedanColor)))
    (:then_1
      (Game|Feedback|ClientPlayForceFeedback _returnvalue (Variables|CatchUpBoost|GetFF_CatchUpBoost) "None" false false false)
      (Projectile|AddGameplayCue%28Looping%29Local _returnvalue (Variables|CatchUpBoost|GetGC_CatchUpBoost) (Ability|GameplayCue|MakeGameplayCueParameters 0.0))
      (CallFunction|ApplyBoostGE)
      (Variables|Camera|SetCameraFOVOffset -1e+01)
      (Utilities|FlowControl|Delay 1.0)
      (Projectile|RemoveGameplayCue%28Looping%29Local (Fortnite|GetFortPlayerControllerfromActor (Variables|ValetDefaults|GetDriverPawn)) (Variables|CatchUpBoost|GetGC_CatchUpBoost) (Ability|GameplayCue|MakeGameplayCueParameters 0.0))
      (Variables|Camera|SetCameraFOVOffset 0.0)
      (CallFunction|CleanUpBoostGE))))

(event EventOnStartHandbrake
  ;[1 1 1] HandBrake ---> Drifting
  (|CallParentFunction)
  (if (Class|FortVehicleInterface|IsTouchingGround)
    (Niagara|SetNiagaraVariable(Bool) (Variables|Valet|RoadAndExhaust|Niagara|GetNativeComp_RoadAndExhaust) "IsDrifting" true)
    (EventDispatchers|BindEventtoOnStartHandBrakeCatchupForceDelegate (EventDispatchers|CreateEvent :!event Custom|OnStartHandBrakeCatchupForceDelegate_Event))
    (CallFunction|ApplyDriftGE)))

(event EventOnPawnEnterVehicle
  (params (PlayerPawn :type Fort%20Player%20Pawn%20Object%20Reference) (SeatIdx :type Integer))
  ;[1 1 1] HandBrake ---> Drifting
  (|CallParentFunction PlayerPawn SeatIdx)
  (CallFunction|InitializeTailLightColor))

(event EventOnPawnExitVehicle
  (params (PlayerPawn :type Fort%20Player%20Pawn%20Object%20Reference) (ExitSocketName :type Name :default "None"))
  ;[1 1 1] HandBrake ---> Drifting
  (|CallParentFunction PlayerPawn ExitSocketName)
  (Niagara|SetNiagaraVariable(Bool) (Variables|Valet|RoadAndExhaust|Niagara|GetNativeComp_RoadAndExhaust) "IsDrifting" false))

(event EventOnStopHandbrake
  ;[1 1 1] HandBrake ---> Drifting
  (|CallParentFunction)
  (CallFunction|CleanUpDriftGE)
  (Niagara|SetNiagaraVariable(Bool) (Variables|Valet|RoadAndExhaust|Niagara|GetNativeComp_RoadAndExhaust) "IsDrifting" false))

(event EventHandleOnExplode
  (params (LastDamageInstigator :type Controller%20Object%20Reference) (Vehicle :type Actor%20Object%20Reference))
  ;[1 1 1] Event ----> Handle Lens Flare on explosion and headlight change
  (|CallParentFunction LastDamageInstigator Vehicle)
  (Components|Activation|Deactivate (Variables|Default|GetNS_DriftSedan_LensFlare)))

(event OnLightEnabledChanged(ValetLightSpawnerComponent)
  (params (enabled :type Boolean))
  (bind _ns_driftsedan_lensflare (Variables|Default|GetNS_DriftSedan_LensFlare))
  ;[1 1 1] Event ----> Handle Lens Flare on explosion and headlight change
  (Components|Activation|SetActive _ns_driftsedan_lensflare enabled true)
  (Rendering|SetVisibility _ns_driftsedan_lensflare enabled false) ; There is a slight delay when deactivating that looks weird so toggling vis
)

(event EventPawnExitSeat
  (params (Vehicle :type Fort%20Vehicle%20Interface%20Interface%20%28by%20ref%29) (PlayerPawn :type Fort%20Pawn%20Object%20Reference) (SeatIndex :type Integer))
  ;[1 1 1] HandBrake ---> Drifting
  (Utilities|FlowControl|Sequence
    (:then_0
      (CallFunction|CleanUpDriftGE)
      (CallFunction|CleanUpBoostGE)
      (Projectile|RemoveGameplayCue%28Looping%29Local (Fortnite|GetFortPlayerControllerfromActor (Variables|ValetDefaults|GetDriverPawn)) (Variables|CatchUpBoost|GetGC_CatchUpBoost) (Ability|GameplayCue|MakeGameplayCueParameters 0.0)))
    (:then_1
      (|CallParentFunction Vehicle PlayerPawn SeatIndex))))

(event EventPickRandomMaterial
  ;[1 1 1] Event ----> Pick Random Material -> Initialize Headlight and Lens Flare FX
  (|CallParentFunction)
  (CallFunction|InitializeHeadlightandLensFlareFX))

(fn ApplyBoostGE
  (GameplayEffects|ApplyGameplayEffectToSelf :self (Ability|GetAbilitySystemComponent (Variables|ValetDefaults|GetDriverPawn)) :GameplayEffectClass (Variables|CatchUpBoost|GetBoostTag) :Level 0.0)
  (for _array_element (Vehicle|GetAllPassengers)
    (Utilities|IsValid _array_element
      (:Is%20Valid
        (GameplayEffects|ApplyGameplayEffectToSelf :self (Ability|GetAbilitySystemComponent _array_element) :GameplayEffectClass (Variables|CatchUpBoost|GetBoostTag) :Level 0.0))))
  (return))

(fn ApplyDriftGE
  (GameplayEffects|ApplyGameplayEffectToSelf :self (Ability|GetAbilitySystemComponent (Variables|ValetDefaults|GetDriverPawn)) :GameplayEffectClass (Variables|Drift|GetDriftingTag) :Level 0.0)
  (for _array_element (Vehicle|GetAllPassengers)
    (Utilities|IsValid _array_element
      (:Is%20Valid
        (GameplayEffects|ApplyGameplayEffectToSelf :self (Ability|GetAbilitySystemComponent _array_element) :GameplayEffectClass (Variables|Drift|GetDriftingTag) :Level 0.0))))
  (return))

(fn CleanupBoostGE
  (GameplayEffects|RemoveActiveGameplayEffectbySourceEffect (Ability|GetAbilitySystemComponent (Variables|ValetDefaults|GetDriverPawn)) (Variables|CatchUpBoost|GetBoostTag) 0 -1)
  (for _array_element (Vehicle|GetAllPassengers)
    (Utilities|IsValid _array_element
      (:Is%20Valid
        (GameplayEffects|RemoveActiveGameplayEffectbySourceEffect (Ability|GetAbilitySystemComponent _array_element) (Variables|CatchUpBoost|GetBoostTag) 0 -1))))
  (return))

(fn CleanupDriftGE
  (GameplayEffects|RemoveActiveGameplayEffectbySourceEffect (Ability|GetAbilitySystemComponent (Variables|ValetDefaults|GetDriverPawn)) (Variables|Drift|GetDriftingTag) 0 -1)
  (for _array_element (Vehicle|GetAllPassengers)
    (Utilities|IsValid _array_element
      (:Is%20Valid
        (GameplayEffects|RemoveActiveGameplayEffectbySourceEffect (Ability|GetAbilitySystemComponent _array_element) (Variables|Drift|GetDriftingTag) 0 -1))))
  (return))

(fn HandleShifting
  (params (NewGear :type Integer))
  (bind _driverpawn (Variables|ValetDefaults|GetDriverPawn))
  (bind _returnvalue (Fortnite|GetFortPlayerControllerfromActor _driverpawn))
  (bind _driverpawn_2 (Variables|ValetDefaults|GetDriverPawn))
  (bind _returnvalue_2 (Fortnite|GetFortPlayerControllerfromActor _driverpawn_2))
  (if (> NewGear (Variables|ShiftingGears|GetCurrentGearAudio))
    (Development|PrintString "UpShift")
    (Game|Feedback|ClientPlayForceFeedback _returnvalue_2 (Variables|ShiftingGears|GetFF_Upshift) "None" false false false)
    (Game|Feedback|ClientStartCameraShake _returnvalue_2 (Variables|ShiftingGears|GetUpShiftCameraShake) 1.0 "CameraLocal" (Transformation|GetActorRotation _driverpawn_2))
    (Variables|ShiftingGears|SetCurrentGearAudio NewGear)
    (return true)
    (else
      (Development|PrintString "Downshift")
      (Game|Feedback|ClientPlayForceFeedback _returnvalue (Variables|ShiftingGears|GetFF_Downshift) "None" false false false)
      (Game|Feedback|ClientStartCameraShake _returnvalue (Variables|ShiftingGears|GetDownShiftCameraShake) 1.0 "CameraLocal" (Transformation|GetActorRotation))
      (Variables|ShiftingGears|SetCurrentGearAudio NewGear)
      (return false))))

(fn InitializeHeadlightAndLensFlareFX
  (bind _result (Math|Integer|IncrementInt (Variables|Lights|GetUpdateValetHeadlightColors))
    (:%20%20
      (Variables|Lights|SetUpdateValetHeadlightColors _result)
      ;[1 1 1] Override value based on child needs
      (Niagara|SetNiagaraVariable(Int32) (Variables|Default|GetNS_DriftSedan_LensFlare) "SedanColor" (Variables|Lights|GetLensFlare_SedanColor)))))

(fn InitializeTailLightColor
  (Niagara|SetNiagaraVariable(Int32) (Variables|Valet|RoadAndExhaust|Niagara|GetNativeComp_RoadAndExhaust) "SedanColor" (Variables|Default|GetRoadAndExhaust_TailLight_SedanColor)))

(fn OnBeginPlayFX
  (Components|Activation|SetActive (Variables|Default|GetNS_DriftSedan_LensFlare) false true) ; Lens Flare should start inactive (until vehicle is occupied)
)

(fn OnRep_UpdateValetHeadlightColors
  (if (not (Networking|IsServer))
    ;[1 1 1] Override value based on child needs
    (ValetLights|SetValetLightColor (Variables|Default|GetValetLightSpawnerComponent) (Variables|Lights|GetHeadLightColor))))

(fn ConstructionScript
  (|CallParentFunction))
